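# Niagara swing trail for the StudyBench pickaxe: layered animation-trail ribbons plus gold sparks.
# Four ribbon emitters follow weapon sockets through a shared Socket Reader, with curves over particle age and emitter loop age shaping width and fade.
# Trail and the two-trail Trail003 each stack four ribbon renderers using separate width/color attributes in blues; Trail001 adds pink, Trail002 gold.
# Sparks spawn around a skeletal mesh socket, inheriting velocity; the Airglow sprite emitter is disabled.
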NiagaraSystem /BRCosmetics/Weapons/FORT_Melee/Pickaxe_StudyBench/FX/NS_Pickaxe_StudyBench_Trail
dependencies {
 usedDataInterfaces [&NiagaraDataInterfaceCurve, &NiagaraDataInterfaceColorCurve, &NiagaraDataInterfaceSocketReader]
 usedDynamicInputs []
 usedModules [
  &SystemState
  &self:SystemSpawnScript.SystemScriptSource.SystemScriptGraph.NiagaraNodeAssignment_0.SetVariables_#1
 ]
}
emitters {
 Airglow {
  bEnabled=false
  emitterUpdateScript [
   &EmitterState(Life Cycle Mode=::NewEnumerator1, Loop Behavior=::NewEnumerator1, Loop Delay=0.03, Loop Duration=0.15, UseLoopDelay=true)
   &SpawnBurst_Instantaneous(Use Spawn Probability=false, UseLoopCountLimit=false)
  ]
  particleSpawnScript [
   &InitializeParticle(
    Color=<rgba 1 0.6198 0.9889 1>
    Lifetime=0.15
    Position Offset=<500 0 0>
    Sprite Size Mode=::NewEnumerator3
    Uniform Sprite Size=40
    UsePositionOffset=false
    Write Material Random=true
   )
   &SpriteFacingAndAlignment(Sprite Facing=<0 -1 0>, WriteAlignment=false, WriteFacing=true) {enabled=false}
   &CameraOffset(Camera Offset Amount=1)
  ]
  particleUpdateScript [
   &ParticleState(Kill Particles When Lifetime Has Elapsed=false)
   &SkeletalMeshLocation(
    Bone / Socket Index=2
    MaskSpawnGroup=false
    Mesh Sampling Type=::NewEnumerator5
    Orientation Sampling=::NewEnumerator1
    Sampled Position Offset=<0 -15 0>
    Skeletal Mesh=$User.SkeletalMesh
    Socket Sampling Mode=::NewEnumerator5
   )
   &ScaleColor(
    Scale Alpha=&FloatFromCurve(
     FloatCurve=DataInterface() {propertyValues {Curve {keys [
         {arriveTangent=-0.0475, interpMode=RCIM_Cubic, leaveTangent=-0.0475, tangentMode=RCTM_User, time=0, value=0}
         {arriveTangent=0.1517, interpMode=RCIM_Cubic, leaveTangent=0.1517, tangentMode=RCTM_User, time=0.15, value=0.9847}
         {arriveTangent=-0.0405, interpMode=RCIM_Cubic, leaveTangent=-0.0405, tangentMode=RCTM_User, time=0.8, value=0}
        ]}}}
     Scale Curve=0.5
    )
    ScaleA=true
    ScaleRGB=true
   )
  ]
  rendererClasses [&NiagaraSpriteRendererProperties]
  renderers [{rendererClass=&NiagaraSpriteRendererProperties}]
 }
 Sparks_World {
  emitterUpdateScript [
   &EmitterState(Life Cycle Mode=::NewEnumerator1, Loop Behavior=::NewEnumerator1, Loop Delay=0.008, Loop Duration=0.2, UseLoopDelay=false)
   &SpawnRate(
    SpawnRate=&FloatFromCurve(
     FloatCurve=DataInterface() {propertyValues {Curve {keys [
         {arriveTangent=0.3706, interpMode=RCIM_Cubic, leaveTangent=0.3706, tangentMode=RCTM_User, time=0.4, value=0.2}
         {arriveTangent=-0.5, interpMode=RCIM_Cubic, leaveTangent=-0.5, time=0.55, value=1}
         {interpMode=RCIM_Cubic, tangentMode=RCTM_User, time=0.8, value=0}
        ]}}}
     CurveIndex=$Emitter.NormalizedLoopAge
     Scale Curve=200
    )
    Use Spawn Probability=false
   )
  ]
  particleSpawnScript [
   &InitializeParticle(
    Color=<rgba 1 0.6739 0 1>
    Lifetime Max=0.4
    Lifetime Min=0.3
    Lifetime Mode=::NewEnumerator1
    Sprite Rotation Mode=::NewEnumerator2
    Sprite Size Mode=::NewEnumerator1
    Uniform Sprite Size Max=6
    Uniform Sprite Size Min=4
    UsePositionOffset=false
    Write Material Random=true
   )
   &SkeletalMeshLocation(
    Bone / Socket Index=2
    MaskSpawnGroup=false
    Mesh Sampling Type=::NewEnumerator5
    Skeletal Mesh=$User.SkeletalMesh
    Socket Sampling Mode=::NewEnumerator5
   ) {module=SkeletalMeshLocation001}
   &SphereLocation(
    MaskSpawnGroup=false
    Override Seed=false
    Sphere Radius=9
    Surface Expansion Mode=::NewEnumerator1
    Surface Only=true
    UseNonUniformScale=false
    UseOffset=true
    UseSphereOrientation=false
   )
   &InheritVelocity(Speed Limit=true, UseSpeedThreshold=false)
  ]
  particleUpdateScript [
   &ParticleState()
   &Drag(Use Linear Drag=true, Use Rotational Drag=true)
   &SolveForcesAndVelocity(Clamp Velocity=false, Limit Acceleration=false)
   &ScaleSpriteSizeBySpeed(Max Scale Factor=<1 1>, Velocity Threshold=100)
   &ScaleSpriteSize(Scale Factor=&Vector2DFromFloat(Value=&FloatFromCurve(FloatCurve=DataInterface() {propertyValues {Curve {keys [
          {arriveTangent=9.0725, interpMode=RCIM_Cubic, leaveTangent=9.0725, tangentMode=RCTM_User, time=0, value=0}
          {arriveTangent=-2.5249, interpMode=RCIM_Cubic, leaveTangent=-2.5249, tangentMode=RCTM_User, time=0.1408, value=0.9847}
          {arriveTangent=-0.0405, interpMode=RCIM_Cubic, leaveTangent=-0.0405, tangentMode=RCTM_User, time=1, value=0}
         ]}}})))
   /SetVariables(Particles.SpriteSize2=&Multiply_Vector2D(
     A=$Particles.SpriteSize
     B=&MakeVector2D(
      X=1
      Y=&FloatFromCurve(FloatCurve=DataInterface() {propertyValues {Curve {keys [
           {interpMode=RCIM_Cubic, tangentMode=RCTM_User, time=0, value=10}
           {arriveTangent=-21.149, interpMode=RCIM_Cubic, leaveTangent=-21.149, tangentMode=RCTM_User, time=1, value=1}
          ]}}})
     )
    )) {isSetParameters=true}
   /SetVariables(Particles.SpriteSize3:Vector2f=&Multiply_Vector2D_ByFloat(
     Float=&FloatFromCurve(FloatCurve=DataInterface() {propertyValues {Curve {keys [
          {arriveTangent=9.0725, interpMode=RCIM_Cubic, leaveTangent=9.0725, tangentMode=RCTM_User, time=0, value=0}
          {arriveTangent=-2.5249, interpMode=RCIM_Cubic, leaveTangent=-2.5249, tangentMode=RCTM_User, time=0.1408, value=0.9847}
          {arriveTangent=-0.0405, interpMode=RCIM_Cubic, leaveTangent=-0.0405, tangentMode=RCTM_User, time=1, value=0}
         ]}}})
     Vector2D=&Vector2DFromFloat(Value=&RandomRangeFloat(Minimum=12.5, Maximum=17.5))
    )) {isSetParameters=true}
   &CameraOffset(Camera Offset Amount=1)
   &DynamicMaterialParameters(
    Index 0 Param 4=10
    Module.Index0Float0Write=false
    Module.Index0Float1Write=false
    Module.Index0Float2Write=false
    Module.Index0Float3Write=true
   )
   &ScaleColor(
    Scale RGB=&VectorFromFloat(Value=&RandomRangeFloat(Minimum=4, Maximum=8))
    ScaleA=true
    ScaleRGB=true
   )
  ]
  rendererClasses [&NiagaraSpriteRendererProperties]
  renderers [{rendererClass=&NiagaraSpriteRendererProperties}]
 }
 Trail {
  emitterUpdateScript [
   &EmitterState(Life Cycle Mode=::NewEnumerator1, Loop Behavior=::NewEnumerator1, Loop Duration=0.4, UseLoopDelay=false)
   &FN_SpawnAnimTrail_v2(Spawn Interval=0)
  ]
  particleSpawnScript [
   /Game/Effects/Niagara/Modules/AnimTrail/FN_InitializeAnimTrail/FN_InitializeAnimTrailBoneIndecies(First Bone or Socket Index=1)
   /Game/Effects/Niagara/Modules/AnimTrail/FN_InitializeAnimTrail/FN_InitializeAnimTrail(AnimTrailWidth=1, Lifetime=0.1) {enabled=false}
   /Game/Effects/Niagara/Modules/AnimTrail/FN_InitializeAnimTrail/FN_InitializeAnimTrail_SocketReader/FN_InitializeAnimTrail_SocketReader(Lifetime=0.22, Socket Reader:NiagaraDataInterfaceSocketReader=$System.Socket Reader)
  ]
  particleUpdateScript [
   &ScaleRibbonWidth(Ribbon Width Scale=&Multiply_Float(
     A=&FloatFromCurve(FloatCurve=DataInterface() {propertyValues {Curve {keys [
          {arriveTangent=-2.4315, interpMode=RCIM_Cubic, leaveTangent=-2.4315, tangentMode=RCTM_User, time=0, value=1}
          {arriveTangent=0.0357, interpMode=RCIM_Cubic, leaveTangent=0.0357, tangentMode=RCTM_User, time=0.35, value=0}
         ]}}})
     B=&FloatFromCurve(
      FloatCurve=DataInterface() {propertyValues {
        bOptimizeLUT=false
        Curve {keys [{interpMode=RCIM_Cubic, tangentMode=RCTM_User, time=0.3, value=1}, {interpMode=RCIM_Cubic, tangentMode=RCTM_User, time=0.49, value=0}]}
       }}
      CurveIndex=$Emitter.NormalizedLoopAge
     )
    )) {module=ScaleRibbonWidth001}
   &ScaleColor(
    Scale Alpha=&Multiply_Float(
     A=&FloatFromCurve(FloatCurve=DataInterface() {propertyValues {Curve {keys [
          {interpMode=RCIM_Cubic, leaveTangent=-5.8353, tangentMode=RCTM_Break, time=0, value=1.5}
          {arriveTangent=-0.3285, interpMode=RCIM_Cubic, leaveTangent=-0.3285, tangentMode=RCTM_User, time=0.6, value=0.15}
          {arriveTangent=-0.2139, interpMode=RCIM_Cubic, tangentMode=RCTM_Break, time=1, value=0}
         ]}}})
     B=1
    )
    Scale RGB=&MakeVectorFromLinearColorRGB(Color=&ColorFromCurve(
      ColorCurve=DataInterface() {propertyValues {
        AlphaCurve {keys [<at 0 1 RCIM_Cubic>]}
        BlueCurve {keys [<at 0 1 RCIM_Cubic>,<at 0.0099 1>,<at 0.2274 0.6979>,<at 1 0.4271 RCIM_Cubic>]}
        GreenCurve {keys [<at 0 0.7651 RCIM_Cubic>,<at 0.0099 0.6087>,<at 0.2274 0.2475>,<at 1 0.0407 RCIM_Cubic>]}
        RedCurve {keys [<at 0 0.3073 RCIM_Cubic>,<at 0.0099 0>,<at 0.2274 0.0073>,<at 1 0.0268 RCIM_Cubic>]}
       }}
      CurveIndex=&Multiply_Float(A=$Particles.NormalizedAge, B=2)
     ))
    ScaleA=true
    ScaleRGB=true
   ) {module=ScaleColor002}
   &FN_UpdateAnimTrail_v2() {enabled=false}
   /Game/Effects/Niagara/Modules/AnimTrail/FN_UpdateAnimTrail/FN_UpdateAnimTrail_SocketReader/FN_UpdateAnimTrail_SocketReader(Socket Reader:NiagaraDataInterfaceSocketReader=$System.Socket Reader)
   /SetVariables(Particles.RibbonWidth02:NiagaraFloat=&Multiply_Float(
     A=4
     B=&Multiply_Float(
      A=&FloatFromCurve(FloatCurve=DataInterface() {propertyValues {Curve {keys [<at 0 1 RCIM_Cubic>, {arriveTangent=0.0357, interpMode=RCIM_Cubic, leaveTangent=0.0357, tangentMode=RCTM_User, time=0.15, value=0}]}}})
      B=&Multiply_Float(
       A=&FloatFromCurve(
        FloatCurve=DataInterface() {propertyValues {
          bOptimizeLUT=false
          Curve {keys [
            <at 0 0 RCIM_Cubic>
            {arriveTangent=0.0903, interpMode=RCIM_Cubic, leaveTangent=0.0903, tangentMode=RCTM_User, time=0.05, value=1}
            {interpMode=RCIM_Cubic, tangentMode=RCTM_User, time=0.3, value=1}
            {interpMode=RCIM_Cubic, tangentMode=RCTM_User, time=0.49, value=0}
           ]}
         }}
        CurveIndex=$Emitter.NormalizedLoopAge
       )
       B=1
      )
     )
    )) {isSetParameters=true, module=SetVariables_#2}
   /SetVariables(Particles.RibbonWidth03:NiagaraFloat=&Multiply_Float(
     A=20
     B=&Multiply_Float(
      A=&FloatFromCurve(FloatCurve=DataInterface() {propertyValues {Curve {keys [
           <at 0 0.5 RCIM_Cubic>
           {arriveTangent=0.0544, interpMode=RCIM_Cubic, leaveTangent=0.0544, tangentMode=RCTM_User, time=0.05, value=1}
           {arriveTangent=-7.1335, interpMode=RCIM_Cubic, leaveTangent=-7.1335, tangentMode=RCTM_User, time=0.15, value=0.5}
           <at 0.25 0 RCIM_Cubic>
          ]}}})
      B=&Multiply_Float(
       A=&FloatFromCurve(
        FloatCurve=DataInterface() {propertyValues {
          bOptimizeLUT=false
          Curve {keys [{interpMode=RCIM_Cubic, tangentMode=RCTM_User, time=0.3, value=1}, {interpMode=RCIM_Cubic, tangentMode=RCTM_User, time=0.49, value=0}]}
         }}
        CurveIndex=$Emitter.NormalizedLoopAge
       )
       B=1
      )
     )
    )) {isSetParameters=true, module=SetVariables_#3}
   /SetVariables(Particles.RibbonWidth04:NiagaraFloat=&Multiply_Float(
     A=16
     B=&Multiply_Float(
      A=&FloatFromCurve(FloatCurve=DataInterface() {propertyValues {Curve {keys [
           {arriveTangent=-0.149, interpMode=RCIM_Cubic, leaveTangent=-0.149, tangentMode=RCTM_User, time=0, value=0}
           <at 0.05 1 RCIM_Cubic>
           {arriveTangent=0.0357, interpMode=RCIM_Cubic, leaveTangent=0.0357, tangentMode=RCTM_User, time=0.25, value=0}
          ]}}})
      B=&Multiply_Float(
       A=&FloatFromCurve(
        FloatCurve=DataInterface() {propertyValues {
          bOptimizeLUT=false
          Curve {keys [{interpMode=RCIM_Cubic, tangentMode=RCTM_User, time=0.3, value=1}, {interpMode=RCIM_Cubic, tangentMode=RCTM_User, time=0.49, value=0}]}
         }}
        CurveIndex=$Emitter.NormalizedLoopAge
       )
       B=1
      )
     )
    )) {isSetParameters=true, module=SetVariables_#4}
   /SetVariables(Particles.Color02:LinearColor=&MultiplyLinearColor(
     Linear Color=<rgba 1 1 1 1>
     Scale Factor=&MakeFloat4FromVectorAndFloat(
      "Vector (RGB)"=&MakeVectorFromLinearColorRGB(Color=&ColorFromCurve(
        ColorCurve=DataInterface() {propertyValues {
          AlphaCurve {keys [<at 0 1 RCIM_Cubic>]}
          BlueCurve {keys [<at 0 1>,<at 0.005 1>,<at 0.0316 0.5677>]}
          GreenCurve {keys [<at 0 0.8711>,<at 0.005 0.6087>,<at 0.0316 0.2071>]}
          RedCurve {keys [<at 0 0.6198>,<at 0.005 0>,<at 0.0316 0.0148>]}
         }}
        CurveIndex=&Multiply_Float(A=$Particles.NormalizedAge, B=2)
       ))
      "Float (Alpha)"=&Multiply_Float(
       A=&FloatFromCurve(FloatCurve=DataInterface() {propertyValues {Curve {keys [
            {interpMode=RCIM_Cubic, leaveTangent=-5.8353, tangentMode=RCTM_Break, time=0, value=1.5}
            {arriveTangent=-0.3285, interpMode=RCIM_Cubic, leaveTangent=-0.3285, tangentMode=RCTM_User, time=0.6, value=0.15}
            {arriveTangent=-0.2139, interpMode=RCIM_Cubic, tangentMode=RCTM_Break, time=1, value=0}
           ]}}})
       B=1
      )
     )
    )) {isSetParameters=true}
   /SetVariables(Particles.Color03:LinearColor=&MultiplyLinearColor(
     Linear Color=<rgba 1 1 1 1>
     Scale Factor=&MakeFloat4FromVectorAndFloat(
      "Vector (RGB)"=&MakeVectorFromLinearColorRGB(Color=&ColorFromCurve(
        ColorCurve=DataInterface() {propertyValues {
          AlphaCurve {keys [<at 0 1 RCIM_Cubic>]}
          BlueCurve {keys [<at 0 1 RCIM_Cubic>,<at 0.0279 1>,<at 0.2274 0.5677>,<at 1 0.4271 RCIM_Cubic>]}
          GreenCurve {keys [<at 0 0.7192 RCIM_Cubic>,<at 0.0279 0.6087>,<at 0.2274 0.2071>,<at 1 0.0407 RCIM_Cubic>]}
          RedCurve {keys [<at 0 0.1719 RCIM_Cubic>,<at 0.0279 0>,<at 0.2274 0.0148>,<at 1 0.0268 RCIM_Cubic>]}
         }}
        CurveIndex=&Multiply_Float(A=$Particles.NormalizedAge, B=2)
       ))
      "Float (Alpha)"=&Multiply_Float(
       A=&FloatFromCurve(
        FloatCurve=DataInterface() {propertyValues {Curve {keys [
            {interpMode=RCIM_Cubic, leaveTangent=0.3823, tangentMode=RCTM_Break, time=0.3, value=1.5}
            {arriveTangent=-0.2139, interpMode=RCIM_Cubic, tangentMode=RCTM_Break, time=1, value=0}
           ]}}}
        Scale Curve=0.5
       )
       B=1
      )
     )
    )) {isSetParameters=true}
   /SetVariables(Particles.Color04:LinearColor=&MultiplyLinearColor(
     Linear Color=<rgba 1 1 1 1>
     Scale Factor=&MakeFloat4FromVectorAndFloat(
      "Vector (RGB)"=&MakeVectorFromLinearColorRGB(Color=&ColorFromCurve(
        ColorCurve=DataInterface() {propertyValues {
          AlphaCurve {keys [<at 0 1 RCIM_Cubic>]}
          BlueCurve {keys [<at 0 1 RCIM_Cubic>,<at 0.1918 1>,<at 0.998 0.4271>]}
          GreenCurve {keys [<at 0 0.7615 RCIM_Cubic>,<at 0.1918 0.6087>,<at 0.998 0.0407>]}
          RedCurve {keys [<at 0 0.2969 RCIM_Cubic>,<at 0.1918 0>,<at 0.998 0.0268>]}
         }}
        CurveIndex=&Multiply_Float(A=$Particles.NormalizedAge, B=2)
       ))
      "Float (Alpha)"=&Multiply_Float(
       A=&FloatFromCurve(
        FloatCurve=DataInterface() {propertyValues {Curve {keys [
            <at 0 0 RCIM_Cubic>
            {arriveTangent=-0.0456, interpMode=RCIM_Cubic, leaveTangent=-0.0456, tangentMode=RCTM_User, time=0.05, value=1.5}
            {arriveTangent=-0.2139, interpMode=RCIM_Cubic, tangentMode=RCTM_Break, time=0.25, value=0}
           ]}}}
        Scale Curve=0.5
       )
       B=1
      )
     )
    )) {isSetParameters=true}
   &DynamicMaterialParameters(Module.Index0Float0Write=true, Module.Index0Float1Write=true, Module.Index0Float2Write=true, Module.Index0Float3Write=true)
  ]
  rendererClasses [&NiagaraRibbonRendererProperties]
  renderers [
   {rendererClass=&NiagaraRibbonRendererProperties}
   {rendererClass=&NiagaraRibbonRendererProperties, rendererIndex=1}
   {rendererClass=&NiagaraRibbonRendererProperties, rendererIndex=2}
   {rendererClass=&NiagaraRibbonRendererProperties, rendererIndex=3}
  ]
 }
 Trail001 {
  emitterUpdateScript [
   &EmitterState(Life Cycle Mode=::NewEnumerator1, Loop Behavior=::NewEnumerator1, Loop Duration=0.4, UseLoopDelay=false)
   &FN_SpawnAnimTrail_v2(Spawn Interval=0)
   /SetVariables(Emitter.Random:NiagaraFloat=&RandomRangeFloat(Minimum=0, Maximum=1)) {isSetParameters=true, module=SetVariables_#5}
  ]
  particleSpawnScript [
   &FN_InitializeAnimTrailBoneIndecies(First Bone or Socket Index=1)
   &FN_InitializeAnimTrail(AnimTrailWidth=1, Lifetime=0.12) {enabled=false}
   &FN_InitializeAnimTrail_SocketReader(Lifetime=0.22, Socket Reader:NiagaraDataInterfaceSocketReader=$System.Socket Reader)
  ]
  particleUpdateScript [
   &ScaleRibbonWidth(Ribbon Width Scale=&Multiply_Float(
     A=&FloatFromCurve(FloatCurve=DataInterface() {propertyValues {Curve {keys [
          {arriveTangent=10.9255, interpMode=RCIM_Cubic, leaveTangent=10.9255, tangentMode=RCTM_User, time=0, value=2}
          {interpMode=RCIM_Cubic, leaveTangent=-4, time=0.35, value=6}
         ]}}})
     B=&FloatFromCurve(
      FloatCurve=DataInterface() {propertyValues {
        bOptimizeLUT=false
        Curve {keys [{interpMode=RCIM_Cubic, tangentMode=RCTM_User, time=0.3, value=1}, {interpMode=RCIM_Cubic, tangentMode=RCTM_User, time=0.52, value=0}]}
       }}
      CurveIndex=$Emitter.NormalizedLoopAge
     )
    ))
   &ScaleColor(
    Scale Alpha=&Multiply_Float(
     A=&FloatFromCurve(
      FloatCurve=DataInterface() {propertyValues {Curve {keys [
          {interpMode=RCIM_Cubic, leaveTangent=-0.2809, tangentMode=RCTM_Break, time=0.25, value=1}
          {arriveTangent=-0.2139, interpMode=RCIM_Cubic, tangentMode=RCTM_Break, time=0.45, value=0}
         ]}}}
      Scale Curve=0.15
     )
     B=1
    )
    Scale RGB=&MakeVectorFromLinearColorRGB(Color=&ColorFromCurve(ColorCurve=DataInterface() {propertyValues {
        AlphaCurve {keys [<at 0 1 RCIM_Cubic>]}
        BlueCurve {keys [<at 0 0.9889 RCIM_Cubic>,<at 0.0279 0.9725>,<at 0.2274 0.5521>,<at 1 0.1376 RCIM_Cubic>]}
        GreenCurve {keys [<at 0 0.6198 RCIM_Cubic>,<at 0.0279 0.0627>,<at 0.2274 0.0356>,<at 1 0.0268 RCIM_Cubic>]}
        RedCurve {keys [<at 0 1 RCIM_Cubic>,<at 0.0279 1>,<at 0.2274 0.5677>,<at 1 0.4271 RCIM_Cubic>]}
       }}))
    ScaleA=true
    ScaleRGB=true
   ) {module=ScaleColor002}
   &FN_UpdateAnimTrail_v2() {enabled=false}
   &FN_UpdateAnimTrail_SocketReader(Socket Reader:NiagaraDataInterfaceSocketReader=$System.Socket Reader)
  ]
  rendererClasses [&NiagaraRibbonRendererProperties]
  renderers [{rendererClass=&NiagaraRibbonRendererProperties}]
 }
 Trail002 {
  emitterUpdateScript [
   &EmitterState(Life Cycle Mode=::NewEnumerator1, Loop Behavior=::NewEnumerator1, Loop Duration=0.4, UseLoopDelay=false)
   &FN_SpawnAnimTrail_v2(Spawn Interval=0)
   /SetVariables(Emitter.Random:NiagaraFloat=&RandomRangeFloat(Minimum=0, Maximum=1)) {isSetParameters=true, module=SetVariables_#5}
  ]
  particleSpawnScript [
   &FN_InitializeAnimTrailBoneIndecies(First Bone or Socket Index=1)
   &FN_InitializeAnimTrail(AnimTrailWidth=0.5, Color=<rgba 10 6.7638 0 1>, Lifetime=0.12) {enabled=false}
   &FN_InitializeAnimTrail_SocketReader(
    AnimTrailWidth=0.5
    Color:LinearColor=<rgba 10 6.7638 0 1>
    Lifetime=0.22
    Socket Reader:NiagaraDataInterfaceSocketReader=$System.Socket Reader
   )
  ]
  particleUpdateScript [
   &ScaleRibbonWidth(Ribbon Width Scale=&Multiply_Float(
     A=&FloatFromCurve(FloatCurve=DataInterface() {propertyValues {Curve {keys [{arriveTangent=3.6271, interpMode=RCIM_Cubic, leaveTangent=3.6271, tangentMode=RCTM_User, time=0, value=3}, <at 0.8 6 RCIM_Cubic>]}}})
     B=&FloatFromCurve(
      FloatCurve=DataInterface() {propertyValues {
        bOptimizeLUT=false
        Curve {keys [{interpMode=RCIM_Cubic, tangentMode=RCTM_User, time=0.3, value=1}, {interpMode=RCIM_Cubic, tangentMode=RCTM_User, time=0.52, value=0.5}]}
       }}
      CurveIndex=$Emitter.NormalizedLoopAge
     )
    ))
   &ScaleColor(
    Scale Alpha=&Multiply_Float(
     A=&FloatFromCurve(
      FloatCurve=DataInterface() {propertyValues {Curve {keys [
          {interpMode=RCIM_Cubic, leaveTangent=-0.2809, tangentMode=RCTM_Break, time=0.5, value=1}
          {arriveTangent=-0.2139, interpMode=RCIM_Cubic, tangentMode=RCTM_Break, time=1, value=0}
         ]}}}
      Scale Curve=2
     )
     B=&FloatFromCurve(
      FloatCurve=DataInterface() {propertyValues {
        bOptimizeLUT=false
        Curve {keys [{interpMode=RCIM_Cubic, tangentMode=RCTM_User, time=0.3, value=1}, {interpMode=RCIM_Cubic, tangentMode=RCTM_User, time=0.45, value=0}]}
       }}
      CurveIndex=$Emitter.NormalizedLoopAge
     )
    )
    ScaleA=true
    ScaleRGB=true
   ) {module=ScaleColor002}
   &FN_UpdateAnimTrail_v2() {enabled=false}
   &FN_UpdateAnimTrail_SocketReader(Socket Reader:NiagaraDataInterfaceSocketReader=$System.Socket Reader)
  ]
  rendererClasses [&NiagaraRibbonRendererProperties]
  renderers [{rendererClass=&NiagaraRibbonRendererProperties}]
 }
 Trail003 {
  emitterUpdateScript [
   &EmitterState(Life Cycle Mode=::NewEnumerator1, Loop Behavior=::NewEnumerator1, Loop Duration=0.4, UseLoopDelay=false)
   &FN_SpawnAnimTrail_v2(Number of Anim Trails=2, Spawn Interval=0)
  ]
  particleSpawnScript [
   &FN_InitializeAnimTrailBoneIndecies(First Bone or Socket Index=4, Second Bone or Socket Index=3)
   &FN_InitializeAnimTrailBoneIndecies(AnimTrail Selection=1, First Bone or Socket Index=6, Second Bone or Socket Index=5) {module=FN_InitializeAnimTrailBoneIndecies001}
   &FN_InitializeAnimTrail(AnimTrailWidth=2.5, Lifetime=0.1) {enabled=false}
   &FN_InitializeAnimTrail_SocketReader(AnimTrailWidth=2.5, Lifetime=0.22, Socket Reader:NiagaraDataInterfaceSocketReader=$System.Socket Reader)
  ]
  particleUpdateScript [
   &ScaleRibbonWidth(Ribbon Width Scale=&Multiply_Float(
     A=&FloatFromCurve(FloatCurve=DataInterface() {propertyValues {Curve {keys [
          {arriveTangent=-2.4315, interpMode=RCIM_Cubic, leaveTangent=-2.4315, tangentMode=RCTM_User, time=0, value=1}
          {arriveTangent=0.0357, interpMode=RCIM_Cubic, leaveTangent=0.0357, tangentMode=RCTM_User, time=0.15, value=0}
         ]}}})
     B=&FloatFromCurve(
      FloatCurve=DataInterface() {propertyValues {
        bOptimizeLUT=false
        Curve {keys [{interpMode=RCIM_Cubic, tangentMode=RCTM_User, time=0.3, value=1}, {interpMode=RCIM_Cubic, tangentMode=RCTM_User, time=0.49, value=0}]}
       }}
      CurveIndex=$Emitter.NormalizedLoopAge
     )
    )) {module=ScaleRibbonWidth001}
   &ScaleColor(
    Scale Alpha=&Multiply_Float(
     A=&FloatFromCurve(FloatCurve=DataInterface() {propertyValues {Curve {keys [
          {interpMode=RCIM_Cubic, leaveTangent=-5.8353, tangentMode=RCTM_Break, time=0, value=1.5}
          {arriveTangent=-0.3285, interpMode=RCIM_Cubic, leaveTangent=-0.3285, tangentMode=RCTM_User, time=0.6, value=0.15}
          {arriveTangent=-0.2139, interpMode=RCIM_Cubic, tangentMode=RCTM_Break, time=1, value=0}
         ]}}})
     B=1
    )
    Scale RGB=&MakeVectorFromLinearColorRGB(Color=&ColorFromCurve(
      ColorCurve=DataInterface() {propertyValues {
        AlphaCurve {keys [<at 0 1 RCIM_Cubic>]}
        BlueCurve {keys [<at 0 1 RCIM_Cubic>,<at 0.0279 1>,<at 0.2274 0.6979>,<at 1 0.4271 RCIM_Cubic>]}
        GreenCurve {keys [<at 0 0.6087 RCIM_Cubic>,<at 0.0279 0.6087>,<at 0.2274 0.2475>,<at 1 0.0407 RCIM_Cubic>]}
        RedCurve {keys [<at 0 0 RCIM_Cubic>,<at 0.0279 0>,<at 0.2274 0.0073>,<at 1 0.0268 RCIM_Cubic>]}
       }}
      CurveIndex=&Multiply_Float(A=$Particles.NormalizedAge, B=2)
     ))
    ScaleA=true
    ScaleRGB=true
   ) {module=ScaleColor002}
   &FN_UpdateAnimTrail_v2() {enabled=false}
   /SetVariables(Particles.RibbonWidth02:NiagaraFloat=&Multiply_Float(
     A=1
     B=&Multiply_Float(
      A=&FloatFromCurve(FloatCurve=DataInterface() {propertyValues {Curve {keys [<at 0 1 RCIM_Cubic>, {arriveTangent=0.0357, interpMode=RCIM_Cubic, leaveTangent=0.0357, tangentMode=RCTM_User, time=0.15, value=0}]}}})
      B=&Multiply_Float(
       A=&FloatFromCurve(
        FloatCurve=DataInterface() {propertyValues {
          bOptimizeLUT=false
          Curve {keys [
            <at 0 0 RCIM_Cubic>
            {arriveTangent=0.0903, interpMode=RCIM_Cubic, leaveTangent=0.0903, tangentMode=RCTM_User, time=0.05, value=1}
            {interpMode=RCIM_Cubic, tangentMode=RCTM_User, time=0.3, value=1}
            {interpMode=RCIM_Cubic, tangentMode=RCTM_User, time=0.49, value=0}
           ]}
         }}
        CurveIndex=$Emitter.NormalizedLoopAge
       )
       B=1
      )
     )
    )) {isSetParameters=true, module=SetVariables_#2}
   /SetVariables(Particles.RibbonWidth03:NiagaraFloat=&Multiply_Float(
     A=5
     B=&Multiply_Float(
      A=&FloatFromCurve(FloatCurve=DataInterface() {propertyValues {Curve {keys [
           <at 0 0.5 RCIM_Cubic>
           {arriveTangent=0.0544, interpMode=RCIM_Cubic, leaveTangent=0.0544, tangentMode=RCTM_User, time=0.05, value=1}
           {arriveTangent=-7.4835, interpMode=RCIM_Cubic, leaveTangent=-7.4835, tangentMode=RCTM_User, time=0.15, value=0.525}
           <at 0.25 0 RCIM_Cubic>
          ]}}})
      B=&Multiply_Float(
       A=&FloatFromCurve(
        FloatCurve=DataInterface() {propertyValues {
          bOptimizeLUT=false
          Curve {keys [{interpMode=RCIM_Cubic, tangentMode=RCTM_User, time=0.3, value=1}, {interpMode=RCIM_Cubic, tangentMode=RCTM_User, time=0.49, value=0}]}
         }}
        CurveIndex=$Emitter.NormalizedLoopAge
       )
       B=1
      )
     )
    )) {isSetParameters=true, module=SetVariables_#3}
   /SetVariables(Particles.RibbonWidth04:NiagaraFloat=&Multiply_Float(
     A=4
     B=&Multiply_Float(
      A=&FloatFromCurve(FloatCurve=DataInterface() {propertyValues {Curve {keys [
           {arriveTangent=-0.149, interpMode=RCIM_Cubic, leaveTangent=-0.149, tangentMode=RCTM_User, time=0, value=0}
           <at 0.05 1 RCIM_Cubic>
           {arriveTangent=0.0357, interpMode=RCIM_Cubic, leaveTangent=0.0357, tangentMode=RCTM_User, time=0.25, value=0}
          ]}}})
      B=&Multiply_Float(
       A=&FloatFromCurve(
        FloatCurve=DataInterface() {propertyValues {
          bOptimizeLUT=false
          Curve {keys [{interpMode=RCIM_Cubic, tangentMode=RCTM_User, time=0.3, value=1}, {interpMode=RCIM_Cubic, tangentMode=RCTM_User, time=0.49, value=0}]}
         }}
        CurveIndex=$Emitter.NormalizedLoopAge
       )
       B=1
      )
     )
    )) {isSetParameters=true, module=SetVariables_#4}
   &DynamicMaterialParameters(Module.Index0Float0Write=true, Module.Index0Float1Write=true, Module.Index0Float2Write=true, Module.Index0Float3Write=true)
   /SetVariables(Particles.Color02:LinearColor=&MultiplyLinearColor(
     Linear Color=<rgba 1 1 1 1>
     Scale Factor=&MakeFloat4FromVectorAndFloat(
      "Vector (RGB)"=&MakeVectorFromLinearColorRGB(Color=&ColorFromCurve(
        ColorCurve=DataInterface() {propertyValues {
          AlphaCurve {keys [<at 0 1 RCIM_Cubic>]}
          BlueCurve {keys [<at 0 1>,<at 0.005 1>,<at 0.0316 0.5677>]}
          GreenCurve {keys [<at 0 0.8711>,<at 0.005 0.6087>,<at 0.0316 0.2071>]}
          RedCurve {keys [<at 0 0.6198>,<at 0.005 0>,<at 0.0316 0.0148>]}
         }}
        CurveIndex=&Multiply_Float(A=$Particles.NormalizedAge, B=2)
       ))
      "Float (Alpha)"=&Multiply_Float(
       A=&FloatFromCurve(FloatCurve=DataInterface() {propertyValues {Curve {keys [
            {interpMode=RCIM_Cubic, leaveTangent=-5.8353, tangentMode=RCTM_Break, time=0, value=1.5}
            {arriveTangent=-0.3285, interpMode=RCIM_Cubic, leaveTangent=-0.3285, tangentMode=RCTM_User, time=0.6, value=0.15}
            {arriveTangent=-0.2139, interpMode=RCIM_Cubic, tangentMode=RCTM_Break, time=1, value=0}
           ]}}})
       B=1
      )
     )
    )) {isSetParameters=true}
   /SetVariables(Particles.Color03:LinearColor=&MultiplyLinearColor(
     Linear Color=<rgba 1 1 1 1>
     Scale Factor=&MakeFloat4FromVectorAndFloat(
      "Vector (RGB)"=&MakeVectorFromLinearColorRGB(Color=&ColorFromCurve(
        ColorCurve=DataInterface() {propertyValues {
          AlphaCurve {keys [<at 0 1 RCIM_Cubic>]}
          BlueCurve {keys [<at 0 1 RCIM_Cubic>,<at 0.0279 1>,<at 0.2274 0.5677>,<at 1 0.4271 RCIM_Cubic>]}
          GreenCurve {keys [<at 0 0.7192 RCIM_Cubic>,<at 0.0279 0.6087>,<at 0.2274 0.2071>,<at 1 0.0407 RCIM_Cubic>]}
          RedCurve {keys [<at 0 0.1719 RCIM_Cubic>,<at 0.0279 0>,<at 0.2274 0.0148>,<at 1 0.0268 RCIM_Cubic>]}
         }}
        CurveIndex=&Multiply_Float(A=$Particles.NormalizedAge, B=2)
       ))
      "Float (Alpha)"=&Multiply_Float(
       A=&FloatFromCurve(
        FloatCurve=DataInterface() {propertyValues {Curve {keys [
            {interpMode=RCIM_Cubic, leaveTangent=0.3823, tangentMode=RCTM_Break, time=0.3, value=1.5}
            {arriveTangent=-0.2139, interpMode=RCIM_Cubic, tangentMode=RCTM_Break, time=1, value=0}
           ]}}}
        Scale Curve=0.5
       )
       B=1
      )
     )
    )) {isSetParameters=true}
   /SetVariables(Particles.Color04:LinearColor=&MultiplyLinearColor(
     Linear Color=<rgba 1 1 1 1>
     Scale Factor=&MakeFloat4FromVectorAndFloat(
      "Vector (RGB)"=&MakeVectorFromLinearColorRGB(Color=&ColorFromCurve(
        ColorCurve=DataInterface() {propertyValues {
          AlphaCurve {keys [<at 0 1 RCIM_Cubic>]}
          BlueCurve {keys [<at 0 1 RCIM_Cubic>,<at 0.1918 1>,<at 0.998 0.4271>]}
          GreenCurve {keys [<at 0 0.7615 RCIM_Cubic>,<at 0.1918 0.6087>,<at 0.998 0.0407>]}
          RedCurve {keys [<at 0 0.2969 RCIM_Cubic>,<at 0.1918 0>,<at 0.998 0.0268>]}
         }}
        CurveIndex=&Multiply_Float(A=$Particles.NormalizedAge, B=2)
       ))
      "Float (Alpha)"=&Multiply_Float(
       A=&FloatFromCurve(
        FloatCurve=DataInterface() {propertyValues {Curve {keys [
            <at 0 0 RCIM_Cubic>
            {arriveTangent=-0.0456, interpMode=RCIM_Cubic, leaveTangent=-0.0456, tangentMode=RCTM_User, time=0.05, value=1.5}
            {arriveTangent=-0.2139, interpMode=RCIM_Cubic, tangentMode=RCTM_Break, time=0.25, value=0}
           ]}}}
        Scale Curve=0.5
       )
       B=1
      )
     )
    )) {isSetParameters=true}
   &FN_UpdateAnimTrail_SocketReader(Socket Reader:NiagaraDataInterfaceSocketReader=$System.Socket Reader)
  ]
  rendererClasses [&NiagaraRibbonRendererProperties]
  renderers [
   {rendererClass=&NiagaraRibbonRendererProperties}
   {rendererClass=&NiagaraRibbonRendererProperties, rendererIndex=1}
   {rendererClass=&NiagaraRibbonRendererProperties, rendererIndex=2}
   {rendererClass=&NiagaraRibbonRendererProperties, rendererIndex=3}
  ]
 }
}
summary {
 systemName=NS_Pickaxe_StudyBench_Trail
 emitters [
  {emitterName=Trail, rendererClasses [&NiagaraRibbonRendererProperties]}
  {emitterName=Trail001, rendererClasses [&NiagaraRibbonRendererProperties]}
  {emitterName=Sparks_World, rendererClasses [&NiagaraSpriteRendererProperties]}
  {emitterName=Trail002, rendererClasses [&NiagaraRibbonRendererProperties]}
  {bEnabled=false, emitterName=Airglow, rendererClasses [&NiagaraSpriteRendererProperties]}
  {emitterName=Trail003, rendererClasses [&NiagaraRibbonRendererProperties]}
 ]
 userVariables [
  User.bIsFrontEnd {type=NiagaraBool}
  User.bIsFrontEndPreview {type=NiagaraBool}
  User.bisSwinging {type=NiagaraBool}
  User.RibbonIDOffset {type=NiagaraInt32}
  User.SkeletalMesh {
   type=NiagaraDataInterfaceSkeletalMesh
   defaultValue {
    struct=NiagaraExt_VariableValue_DataInterface
    value {dataInterface=&self:NiagaraDataInterfaceSkeletalMesh_0}
   }
  }
 ]
}
