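# Physics asset for the StudyBench pickaxe mesh, a two-handed melee weapon cosmetic.
# It holds a single body on the Melee_TwoHanded_Handle bone, shaped as one large capsule that is offset and rotated from that bone.
# It defines no constraints, so there are no jointed parts.
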PhysicsAsset /BRCosmetics/Weapons/FORT_Melee/Pickaxe_StudyBench/Meshes/StudyBench_Axe_PhysicsAsset
bodies {Melee_TwoHanded_Handle [{shapeType=Capsule, length=61.6834, radius=39.0146, center=<0.7653 -6.7825 32.8853>, rotation=<rot 0 -1.2691 -155.3062>}]}
constraints []
